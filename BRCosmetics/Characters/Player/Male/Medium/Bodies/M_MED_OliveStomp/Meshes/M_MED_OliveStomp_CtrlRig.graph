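# Control Rig for the OliveStomp medium male body cosmetic that adds pose-driven corrective deformation to helper bones.
# The main graph runs a shared DeformRig function, then per-side functions for heels, shoulder pads, elbows, belts, wrists, body sides and knees.
# Spherical pose readers on driver bones weight transform offsets, with side flags negating vectors for mirroring.
# Constraints place knee base bones and position and aim the middle belt bones.
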ControlRigBlueprint /BRCosmetics/Characters/Player/Male/Medium/Bodies/M_MED_OliveStomp/Meshes/M_MED_OliveStomp_CtrlRig
graphs [
 New Function_ContainedGraph {nodes [
   n1 = RigVMFunctionEntryNode()
   RigVMFunction_ControlFlowBranch = &RigVMFunction_ControlFlowBranch(ExecuteContext=@n1.ExecuteContext, Condition=@n1.Left)
   Multiply = /Script/RigVM.RigVMFunction_MathVectorMul(B=(X=-1,Y=-1,Z=-1), A=RigVMVariableNode() {variable=BaseDriveTranslate}.Value) {is=struct RigVMUnitNode, template="Multiply::Execute(in A,in B,out Result)"}
   RigVMVariableNode(Value=@Multiply.Result, ExecuteContext=@RigVMFunction_ControlFlowBranch.True) {variable=BaseDriveTranslate}
   > RigVMVariableNode(Value=&RigVMFunction_MathVectorMul(B=(X=-1,Y=-1,Z=-1), A=RigVMVariableNode() {variable=DriveTranslate}.Value) {template="Multiply::Execute(in A,in B,out Result)"}.Result) {variable=DriveTranslate}
   n2 = &RigUnit_SphericalPoseReader(
    ActiveRegionScaleFactors=(PositiveWidth=1,NegativeWidth=1,PositiveHeight=1,NegativeHeight=1)
    ActiveRegionSize=0.17
    Debug=(bDrawDebug=false,bDraw2D=False,bDrawLocalAxes=False,DebugScale=25,DebugSegments=20,DebugThickness=0.25)
    DriverAxis=(X=1,Y=0,Z=0)
    FalloffRegionScaleFactors=(PositiveWidth=1,NegativeWidth=1,PositiveHeight=1,NegativeHeight=1)
    FalloffSize=0.5
    FlipHeightScaling=False
    FlipWidthScaling=False
    RotationOffset=(X=90,Y=0,Z=0)
    DriverItem=RigVMVariableNode() {variable=driverBone}.Value
    OptionalParentItem=RigVMVariableNode() {variable=boneBaseDeform}.Value
    ExecutePin=@RigVMFunction_ControlFlowBranch.Completed
   )
   > &RigUnit_ModifyTransforms(
    ItemToModify=((Item=(Type=Bone,Name="None"),Transform=(Rotation=(X=0,Y=0,Z=0,W=1),Translation=(X=0,Y=3.5,Z=0),Scale3D=(X=1,Y=1,Z=1))))
    WeightMinimum=0
    Weight=@n2.OutputParam
    ItemToModify.0.Item=RigVMVariableNode() {variable=boneBaseDeform}.Value
    ItemToModify.0.Transform.Translation=RigVMVariableNode() {variable=BaseDriveTranslate}.Value
   )
   > &RigUnit_ModifyTransforms(
    ItemToModify=((Item=(Type=Bone,Name="None"),Transform=(Rotation=(X=0,Y=0,Z=0,W=1),Translation=(X=-9,Y=0,Z=0),Scale3D=(X=1,Y=1,Z=1))))
    WeightMinimum=0
    ItemToModify.0.Item=RigVMVariableNode() {variable=boneDeform}.Value
    ItemToModify.0.Transform.Translation=RigVMVariableNode() {variable=DriveTranslate}.Value
    Weight=@n2.OutputParam
   )
   > RigVMFunctionReturnNode()
  ]}
 New Function_ContainedGraph {nodes [
   n1 = RigVMFunctionEntryNode()
   RigVMFunction_ControlFlowBranch = &RigVMFunction_ControlFlowBranch(ExecuteContext=@n1.ExecuteContext, Condition=@n1.Right)
   RigVMVariableNode(Value=0.4, ExecuteContext=@RigVMFunction_ControlFlowBranch.False) {variable=ForArm_FallofPosHeight}
   RigVMVariableNode(Value=1, ExecuteContext=@RigVMFunction_ControlFlowBranch.True) {variable=ForArm_FallofPosHeight}
   > RigVMVariableNode(Value=&RigVMFunction_MathVectorMul(B=(X=1,Y=-1,Z=1), A=RigVMVariableNode() {variable=AimVectorDown}.Value) {template="Multiply::Execute(in A,in B,out Result)"}.Result) {variable=AimVectorDown}
   > RigVMVariableNode(Value=&RigVMFunction_MathVectorMul(B=(X=-1,Y=-1,Z=-1), A=RigVMVariableNode() {variable=AimAxes}.Value) {template="Multiply::Execute(in A,in B,out Result)"}.Result) {variable=AimAxes}
   > RigVMVariableNode(Value=&RigVMFunction_MathVectorMul(B=(X=-1,Y=-1,Z=-1), A=RigVMVariableNode() {variable=DriverAxes}.Value) {template="Multiply::Execute(in A,in B,out Result)"}.Result) {variable=DriverAxes}
   > RigVMVariableNode(Value=(X=-1,Y=-1,Z=-1)) {variable=UpArm_Reverser}
   > RigVMVariableNode(Value=&RigVMFunction_MathVectorMul(B=(X=-1,Y=1,Z=-1), A=RigVMVariableNode() {variable=UpArm_RotationSphere}.Value) {template="Multiply::Execute(in A,in B,out Result)"}.Result) {variable=UpArm_RotationSphere}
   > RigVMVariableNode(Value=&RigVMFunction_MathVectorMul(B=(X=1,Y=1,Z=1), A=RigVMVariableNode() {variable=ForwardArm_RotSph}.Value) {template="Multiply::Execute(in A,in B,out Result)"}.Result) {variable=ForwardArm_RotSph}
   > RigVMVariableNode(Value=(X=-1,Y=-1,Z=-1)) {variable=ForArm_Reverser}
   > RigVMVariableNode(Value=&RigVMFunction_MathVectorMul(B=(X=-1,Y=-1,Z=-1), A=RigVMVariableNode() {variable=BackArm_RotSph}.Value) {template="Multiply::Execute(in A,in B,out Result)"}.Result) {variable=BackArm_RotSph}
   n2 = RigVMVariableNode() {variable=BackwardArm_Transform}
   > RigVMVariableNode(
    Value=(Rotation=(X=0.1285,Y=0.1722,Z=-0.0227,W=0.9764),Translation=(X=0,Y=0,Z=-1),Scale3D=(X=1,Y=1,Z=1))
    Value.Translation=&RigVMFunction_MathVectorMul(B=(X=-1,Y=-1,Z=-1), A=@n2.Value.Translation) {template="Multiply::Execute(in A,in B,out Result)"}.Result
    Value.Rotation=@n2.Value.Rotation
   ) {variable=BackwardArm_Transform}
   > RigVMVariableNode(Value=&RigVMFunction_MathVectorMul(B=(X=-1,Y=-1,Z=-1), A=RigVMVariableNode() {variable=downArm_RotSph}.Value) {template="Multiply::Execute(in A,in B,out Result)"}.Result) {variable=downArm_RotSph}
   > RigVMVariableNode(Value=(X=-1,Y=-1,Z=-1)) {variable=DownArm_Reverser}
   n3 = &RigUnit_SphericalPoseReader(
    ActiveRegionScaleFactors=(PositiveWidth=1,NegativeWidth=1,PositiveHeight=0.9,NegativeHeight=0.9)
    ActiveRegionSize=0.2
    Debug=(bDrawDebug=false,bDraw2D=false,bDrawLocalAxes=False,DebugScale=25,DebugSegments=20,DebugThickness=0.1)
    DriverAxis=(X=1,Y=0,Z=0)
    FalloffRegionScaleFactors=(PositiveWidth=1,NegativeWidth=1,PositiveHeight=0.5,NegativeHeight=0.5)
    FalloffSize=0.6
    FlipHeightScaling=False
    FlipWidthScaling=False
    DriverItem=RigVMVariableNode() {variable=DriverBone}.Value
    OptionalParentItem=RigVMVariableNode() {variable=claviceBone}.Value
    RotationOffset=RigVMVariableNode() {variable=UpArm_RotationSphere}.Value
    ExecutePin=@RigVMFunction_ControlFlowBranch.Completed
   )
   Multiply = &RigVMFunction_MathVectorMul(B=(X=-1,Y=0,Z=-2), A=RigVMVariableNode() {variable=UpArm_Reverser}.Value) {template="Multiply::Execute(in A,in B,out Result)"}
   > &RigUnit_ModifyTransforms(
    ItemToModify=((Item=(Type=Bone,Name="None"),Transform=(Rotation=(X=0,Y=0.4226,Z=0,W=0.9063),Translation=(X=-14,Y=0,Z=0),Scale3D=(X=1,Y=1,Z=1))))
    WeightMinimum=0
    ItemToModify.0.Transform.Translation=@Multiply.Result
    Weight=@n3.OutputParam
    ItemToModify.0.Item=RigVMVariableNode() {variable=DeformBone}.Value
   )
   > n4 = &RigUnit_SphericalPoseReader(
    ActiveRegionScaleFactors=(PositiveWidth=1,NegativeWidth=1,PositiveHeight=0.9,NegativeHeight=0.9)
    ActiveRegionSize=0.2
    Debug=(bDrawDebug=false,bDraw2D=false,bDrawLocalAxes=False,DebugScale=25,DebugSegments=20,DebugThickness=0.1)
    DriverAxis=(X=1,Y=0,Z=0)
    FalloffRegionScaleFactors=(PositiveWidth=1,NegativeWidth=1,PositiveHeight=0.5,NegativeHeight=0.5)
    FalloffSize=0.6
    FlipHeightScaling=False
    FlipWidthScaling=False
    RotationOffset=(X=0,Y=48,Z=0)
    DriverItem=RigVMVariableNode() {variable=DriverBone}.Value
    OptionalParentItem=RigVMVariableNode() {variable=spineBone}.Value
   )
   > &RigUnit_ModifyTransforms(
    ItemToModify=((Item=(Type=Bone,Name="None"),Transform=(Rotation=(X=0,Y=0.5736,Z=0,W=0.8192),Translation=(X=-14,Y=0,Z=0),Scale3D=(X=1,Y=1,Z=1))))
    Mode=OverrideLocal
    WeightMinimum=0
    ItemToModify.0.Transform.Translation=&RigVMFunction_MathVectorMul(B=(X=1,Y=0,Z=1), A=RigVMVariableNode() {variable=UpArm_Reverser}.Value) {template="Multiply::Execute(in A,in B,out Result)"}.Result
    Weight=@n4.OutputParam
    ItemToModify.0.Item=RigVMVariableNode() {variable=DeformBone}.Value
   )
   > n5 = &RigUnit_SphericalPoseReader(
    ActiveRegionScaleFactors=(PositiveWidth=1,NegativeWidth=1,PositiveHeight=1,NegativeHeight=1)
    ActiveRegionSize=0.24
    Debug=(bDrawDebug=false,bDraw2D=False,bDrawLocalAxes=False,DebugScale=25,DebugSegments=20,DebugThickness=0.1)
    DriverAxis=(X=1,Y=0,Z=0)
    FalloffRegionScaleFactors=(PositiveWidth=1,NegativeWidth=1,PositiveHeight=0.7,NegativeHeight=0.7)
    FalloffSize=0.5
    FlipHeightScaling=False
    FlipWidthScaling=False
    DriverItem=RigVMVariableNode() {variable=DriverBone}.Value
    OptionalParentItem=RigVMVariableNode() {variable=claviceBone}.Value
    RotationOffset=RigVMVariableNode() {variable=downArm_RotSph}.Value
   )
   > &RigUnit_ModifyTransforms(
    ItemToModify=((Item=(Type=Bone,Name="None"),Transform=(Rotation=(X=-0.171,Y=-0.171,Z=-0.0302,W=0.9698),Translation=(X=0,Y=0,Z=0),Scale3D=(X=1,Y=1,Z=1))))
    WeightMinimum=0
    ItemToModify.0.Transform.Translation=&RigVMFunction_MathVectorMul(B=(X=0,Y=0,Z=0), A=RigVMVariableNode() {variable=DownArm_Reverser}.Value) {template="Multiply::Execute(in A,in B,out Result)"}.Result
    ItemToModify.0.Transform.Rotation=&RigVMFunction_MathQuaternionFromEuler(Euler=(X=-10,Y=-20,Z=-20)).Result
    Weight=@n5.OutputParam
    ItemToModify.0.Item=RigVMVariableNode() {variable=DeformBone}.Value
   )
   > n6 = &RigUnit_SphericalPoseReader(
    ActiveRegionScaleFactors=(PositiveWidth=1,NegativeWidth=1,PositiveHeight=1,NegativeHeight=1)
    ActiveRegionSize=0.02
    Debug=(bDrawDebug=true,bDraw2D=False,bDrawLocalAxes=False,DebugScale=25,DebugSegments=20,DebugThickness=0.1)
    DriverAxis=(X=1,Y=0,Z=0)
    FalloffRegionScaleFactors=(PositiveWidth=1,NegativeWidth=1,PositiveHeight=0.5,NegativeHeight=0.5)
    FalloffSize=0.5
    FlipHeightScaling=False
    FlipWidthScaling=False
    RotationOffset=(X=100,Y=0,Z=0)
    OptionalParentItem=RigVMVariableNode() {variable=claviceBone}.Value
    DriverItem=RigVMVariableNode() {variable=DriverBone}.Value
   )
   > &RigUnit_ModifyTransforms(
    ItemToModify=((Item=(Type=Bone,Name="None"),Transform=(Rotation=(X=-0.4995,Y=-0.0218,Z=0.0378,W=0.8652),Translation=(X=0,Y=0,Z=0),Scale3D=(X=1,Y=1.3,Z=1))))
    WeightMinimum=0
    ItemToModify.0.Transform.Translation=&RigVMFunction_MathVectorMul(B=(X=0,Y=2.5,Z=-1.5), A=RigVMVariableNode() {variable=ForArm_Reverser}.Value) {template="Multiply::Execute(in A,in B,out Result)"}.Result
    ItemToModify.0.Transform.Rotation=&RigVMFunction_MathQuaternionFromEuler(Euler=(X=-36,Y=20,Z=-25)).Result
    Weight=@n6.OutputParam
    ItemToModify.0.Item=RigVMVariableNode() {variable=DeformBone}.Value
   )
   > n7 = &RigUnit_SphericalPoseReader(
    ActiveRegionScaleFactors=(PositiveWidth=1,NegativeWidth=1,PositiveHeight=1,NegativeHeight=1)
    ActiveRegionSize=0.15
    Debug=(bDrawDebug=true,bDraw2D=False,bDrawLocalAxes=False,DebugScale=25,DebugSegments=20,DebugThickness=0.1)
    DriverAxis=(X=1,Y=0,Z=0)
    FalloffRegionScaleFactors=(PositiveWidth=0.4,NegativeWidth=1,PositiveHeight=1,NegativeHeight=1)
    FalloffSize=0.35
    FlipHeightScaling=False
    FlipWidthScaling=False
    RotationOffset=(X=0,Y=40,Z=-35)
    OptionalParentItem=RigVMVariableNode() {variable=claviceBone}.Value
    FalloffRegionScaleFactors.PositiveHeight=RigVMVariableNode() {variable=ForArm_FallofPosHeight}.Value
    DriverItem=RigVMVariableNode() {variable=DriverBone}.Value
   )
   RigVMFunction_MathQuaternionFromEuler = &RigVMFunction_MathQuaternionFromEuler(Euler=(X=-35,Y=30,Z=-10), RotationOrder=ZXY)
   > &RigUnit_ModifyTransforms(
    ItemToModify=((Item=(Type=Bone,Name="None"),Transform=(Rotation=(X=-0.2588,Y=0,Z=0,W=0.9659),Translation=(X=0,Y=0,Z=0),Scale3D=(X=1,Y=1.1,Z=1))))
    Mode=OverrideLocal
    WeightMinimum=0
    ItemToModify.0.Transform.Translation=&RigVMFunction_MathVectorMul(B=(X=1,Y=2,Z=0), A=RigVMVariableNode() {variable=ForArm_Reverser}.Value) {template="Multiply::Execute(in A,in B,out Result)"}.Result
    ItemToModify.0.Transform.Rotation=@RigVMFunction_MathQuaternionFromEuler.Result
    Weight=@n7.OutputParam
    ItemToModify.0.Item=RigVMVariableNode() {variable=DeformBone}.Value
   )
   > RigVMFunctionReturnNode()
   n8 = &RigUnit_SphericalPoseReader(
    ActiveRegionScaleFactors=(PositiveWidth=1,NegativeWidth=1,PositiveHeight=1,NegativeHeight=1)
    ActiveRegionSize=0.1
    Debug=(bDrawDebug=false,bDraw2D=False,bDrawLocalAxes=False,DebugScale=25,DebugSegments=20,DebugThickness=0.15)
    ExecutePin=()
    FalloffRegionScaleFactors=(PositiveWidth=1,NegativeWidth=1,PositiveHeight=0.3,NegativeHeight=0.3)
    FalloffSize=0.5
    FlipHeightScaling=False
    FlipWidthScaling=False
    DriverItem=RigVMVariableNode() {variable=DriverBone}.Value
    OptionalParentItem=RigVMVariableNode() {variable=claviceBone}.Value
    DriverAxis=RigVMVariableNode() {variable=DriverAxes}.Value
    RotationOffset=RigVMVariableNode() {variable=BackArm_RotSph}.Value
   )
   > &RigUnit_ModifyTransforms(
    ItemToModify=((Item=(Type=Bone,Name="None"),Transform=(Rotation=(X=0,Y=0,Z=0,W=1),Translation=(X=0,Y=0,Z=0),Scale3D=(X=1,Y=1,Z=1))))
    WeightMinimum=0
    Weight=@n8.OutputParam
    ItemToModify.0.Transform=RigVMVariableNode() {variable=BackwardArm_Transform}.Value
    ItemToModify.0.Item=RigVMVariableNode() {variable=DeformBone}.Value
   )
   n9 = &RigUnit_SphericalPoseReader(
    ActiveRegionScaleFactors=(PositiveWidth=0.5,NegativeWidth=0.5,PositiveHeight=0.5,NegativeHeight=2)
    ActiveRegionSize=0.1
    Debug=(bDrawDebug=true,bDraw2D=False,bDrawLocalAxes=False,DebugScale=25,DebugSegments=20,DebugThickness=0.25)
    DriverAxis=(X=1,Y=0,Z=0)
    ExecutePin=()
    FalloffRegionScaleFactors=(PositiveWidth=0.5,NegativeWidth=0.5,PositiveHeight=1,NegativeHeight=1)
    FalloffSize=0.4
    FlipHeightScaling=False
    FlipWidthScaling=False
    RotationOffset=(X=30,Y=30,Z=0)
    DriverItem=RigVMVariableNode() {variable=DriverBone}.Value
    OptionalParentItem=RigVMVariableNode() {variable=claviceBone}.Value
   )
   > &RigUnit_ModifyTransforms(
    ItemToModify=((Item=(Type=Bone,Name="None"),Transform=(Rotation=(X=-0.4226,Y=0,Z=0,W=0.9063),Translation=(X=-5,Y=-15,Z=0),Scale3D=(X=1,Y=1,Z=1))))
    WeightMinimum=0
    Weight=@n9.OutputParam
    ItemToModify.0.Item=RigVMVariableNode() {variable=DeformBone}.Value
   )
   RigVMVariableNode() {variable=DriverAxes}
   RigVMVariableNode() {variable=DriverAxes}
   RigVMVariableNode() {variable=ForwardArm_RotSph}
   RigVMVariableNode() {variable=ForwardArm_RotSph}
   RigVMVariableNode() {variable=UpArm_RotationSphere}
  ]}
 New Function_ContainedGraph {nodes [
   n1 = RigVMFunctionEntryNode()
   RigVMFunction_ControlFlowBranch = &RigVMFunction_ControlFlowBranch(ExecuteContext=@n1.ExecuteContext, Condition=@n1.Right)
   n2 = &RigUnit_SphericalPoseReader(
    ActiveRegionScaleFactors=(PositiveWidth=1,NegativeWidth=2,PositiveHeight=1,NegativeHeight=1)
    ActiveRegionSize=0.1
    Debug=(bDrawDebug=false,bDraw2D=False,bDrawLocalAxes=False,DebugScale=25,DebugSegments=20,DebugThickness=0.1)
    DriverAxis=(X=1,Y=0,Z=0)
    FalloffRegionScaleFactors=(PositiveWidth=1,NegativeWidth=1,PositiveHeight=0.8,NegativeHeight=0.8)
    FalloffSize=0.6
    FlipHeightScaling=False
    FlipWidthScaling=False
    RotationOffset=(X=90,Y=0,Z=18)
    DriverItem=RigVMVariableNode() {variable=DriverBone}.Value
    OptionalParentItem=RigVMVariableNode() {variable=parentBone}.Value
    ExecutePin=@RigVMFunction_ControlFlowBranch.Completed
   )
   > &RigUnit_ModifyTransforms(
    ItemToModify=((Item=(Type=Bone,Name="None"),Transform=(Rotation=(X=0,Y=0,Z=-0.1305,W=0.9914),Translation=(X=0,Y=0,Z=0),Scale3D=(X=1,Y=1,Z=1))))
    WeightMinimum=0
    ItemToModify.0.Item=RigVMVariableNode() {variable=BaseDone}.Value
    Weight=@n2.OutputParam
   )
   > RigVMFunctionReturnNode()
  ]}
 New Function_ContainedGraph {nodes [
   n1 = RigVMFunctionEntryNode()
   RigVMFunction_ControlFlowBranch = &RigVMFunction_ControlFlowBranch(ExecuteContext=@n1.ExecuteContext, Condition=@n1.Right)
   Multiply = &RigVMFunction_MathVectorMul(B=(X=-1,Y=-1,Z=-1), A=RigVMVariableNode() {variable=beltTranslate}.Value) {template="Multiply::Execute(in A,in B,out Result)"}
   RigVMVariableNode(Value=@Multiply.Result, ExecuteContext=@RigVMFunction_ControlFlowBranch.True) {variable=beltTranslate}
   > RigVMVariableNode(Value=(X=-1,Y=0,Z=0)) {variable=driverAxes}
   n2 = &RigUnit_SphericalPoseReader(
    ActiveRegionScaleFactors=(PositiveWidth=1,NegativeWidth=1,PositiveHeight=1,NegativeHeight=1)
    ActiveRegionSize=0.1
    Debug=(bDrawDebug=false,bDraw2D=False,bDrawLocalAxes=False,DebugScale=25,DebugSegments=20,DebugThickness=0.1)
    FalloffRegionScaleFactors=(PositiveWidth=1,NegativeWidth=1,PositiveHeight=0.9,NegativeHeight=0.9)
    FalloffSize=0.5
    FlipHeightScaling=False
    FlipWidthScaling=False
    RotationOffset=(X=90,Y=0,Z=-60)
    DriverItem=RigVMVariableNode() {variable=thigh}.Value
    OptionalParentItem=RigVMVariableNode() {variable=Pelvis}.Value
    DriverAxis=RigVMVariableNode() {variable=driverAxes}.Value
    ExecutePin=@RigVMFunction_ControlFlowBranch.Completed
   )
   > &RigUnit_ModifyTransforms(
    ItemToModify=((Item=(Type=Bone,Name="None"),Transform=(Rotation=(X=0,Y=0,Z=0,W=1),Translation=(X=0,Y=0,Z=0),Scale3D=(X=1,Y=1,Z=1))))
    WeightMinimum=0
    ItemToModify.0.Item=RigVMVariableNode() {variable=beltBone}.Value
    ItemToModify.0.Transform.Translation=RigVMVariableNode() {variable=beltTranslate}.Value
    Weight=@n2.OutputParam
   )
   > RigVMFunctionReturnNode()
  ]}
 New Function_ContainedGraph {nodes [
   n1 = RigVMFunctionEntryNode()
   RigVMFunction_ControlFlowBranch = &RigVMFunction_ControlFlowBranch(ExecuteContext=@n1.ExecuteContext, Condition=@n1.Right)
   Multiply = &RigVMFunction_MathVectorMul(B=(X=-1,Y=-1,Z=-1), A=RigVMVariableNode() {variable=secBoneTranslation}.Value) {template="Multiply::Execute(in A,in B,out Result)"}
   RigVMVariableNode(Value=@Multiply.Result, ExecuteContext=@RigVMFunction_ControlFlowBranch.True) {variable=secBoneTranslation}
   n2 = &RigUnit_SphericalPoseReader(
    ActiveRegionScaleFactors=(PositiveWidth=1,NegativeWidth=3,PositiveHeight=5,NegativeHeight=5)
    ActiveRegionSize=0.1
    Debug=(bDrawDebug=false,bDraw2D=False,bDrawLocalAxes=False,DebugScale=25,DebugSegments=20,DebugThickness=0.25)
    DriverAxis=(X=1,Y=0,Z=0)
    FalloffRegionScaleFactors=(PositiveWidth=1,NegativeWidth=1,PositiveHeight=1,NegativeHeight=1)
    FalloffSize=0.6
    FlipHeightScaling=False
    FlipWidthScaling=False
    RotationOffset=(X=90,Y=0,Z=30)
    DriverItem=RigVMVariableNode() {variable=handBone}.Value
    OptionalParentItem=RigVMVariableNode() {variable=ElbowBone}.Value
    ExecutePin=@RigVMFunction_ControlFlowBranch.Completed
   )
   > &RigUnit_ModifyTransforms(
    ItemToModify=((Item=(Type=Bone,Name="None"),Transform=(Rotation=(X=0,Y=0,Z=0.1736,W=0.9848),Translation=(X=0,Y=0,Z=0),Scale3D=(X=1,Y=1,Z=1))))
    WeightMinimum=0
    Weight=@n2.OutputParam
    ItemToModify.0.Item=RigVMVariableNode() {variable=DeformBone}.Value
   )
   > &RigUnit_ModifyTransforms(
    ItemToModify=((Item=(Type=Bone,Name="None"),Transform=(Rotation=(X=0,Y=0,Z=0,W=1),Translation=(X=1.5,Y=9.5,Z=0),Scale3D=(X=1,Y=1,Z=1))))
    WeightMinimum=0
    ItemToModify.0.Item=RigVMVariableNode() {variable=DeformBone_2}.Value
    ItemToModify.0.Transform.Translation=RigVMVariableNode() {variable=secBoneTranslation}.Value
    Weight=@n2.OutputParam
   )
   > RigVMFunctionReturnNode()
  ]}
 New Function_ContainedGraph {nodes [
   n1 = RigVMFunctionEntryNode()
   RigVMFunction_ControlFlowBranch = &RigVMFunction_ControlFlowBranch(ExecuteContext=@n1.ExecuteContext, Condition=@n1.Right)
   RigVMVariableNode(Value=(X=-1,Y=-1,Z=1), ExecuteContext=@RigVMFunction_ControlFlowBranch.True) {variable=deformBone_Reverse}
   n2 = &RigUnit_SphericalPoseReader(
    ActiveRegionScaleFactors=(PositiveWidth=1,NegativeWidth=1,PositiveHeight=1,NegativeHeight=1)
    ActiveRegionSize=0.1
    Debug=(bDrawDebug=true,bDraw2D=False,bDrawLocalAxes=False,DebugScale=15,DebugSegments=20,DebugThickness=0.1)
    DriverAxis=(X=0,Y=1,Z=0)
    FalloffRegionScaleFactors=(PositiveWidth=1,NegativeWidth=0.83,PositiveHeight=1,NegativeHeight=1)
    FalloffSize=0.4
    FlipHeightScaling=False
    FlipWidthScaling=False
    RotationOffset=(X=90,Y=0,Z=-63)
    DriverItem=RigVMVariableNode() {variable=footBone}.Value
    OptionalParentItem=RigVMVariableNode() {variable=legBone}.Value
    ExecutePin=@RigVMFunction_ControlFlowBranch.Completed
   )
   Multiply = &RigVMFunction_MathVectorMul(B=(X=6,Y=5,Z=0), A=RigVMVariableNode() {variable=deformBone_Reverse}.Value) {template="Multiply::Execute(in A,in B,out Result)"}
   > &RigUnit_ModifyTransforms(
    ItemToModify=((Item=(Type=Bone,Name="None"),Transform=(Rotation=(X=0,Y=0,Z=0,W=1),Translation=(X=6,Y=5,Z=0),Scale3D=(X=1,Y=1,Z=1))))
    WeightMinimum=0
    ItemToModify.0.Transform.Translation=@Multiply.Result
    ItemToModify.0.Item=RigVMVariableNode() {variable=DeformBone}.Value
    Weight=@n2.OutputParam
   )
   > RigVMFunctionReturnNode()
  ]}
 New Function_ContainedGraph {nodes [
   n1 = RigVMFunctionEntryNode()
   RigVMFunction_ControlFlowBranch = &RigVMFunction_ControlFlowBranch(ExecuteContext=@n1.ExecuteContext, Condition=@n1.Right)
   RigVMVariableNode(Value=(X=1,Y=1,Z=1), ExecuteContext=@RigVMFunction_ControlFlowBranch.False) {variable=reverseTran}
   RigVMVariableNode(Value=(X=-1,Y=-1,Z=-1), ExecuteContext=@RigVMFunction_ControlFlowBranch.True) {variable=reverseTran}
   n2 = &RigUnit_SphericalPoseReader(
    ActiveRegionScaleFactors=(PositiveWidth=1,NegativeWidth=1,PositiveHeight=1,NegativeHeight=1)
    ActiveRegionSize=0
    Debug=(bDrawDebug=false,bDraw2D=False,bDrawLocalAxes=False,DebugScale=25,DebugSegments=20,DebugThickness=0.25)
    DriverAxis=(X=1,Y=0,Z=0)
    FalloffRegionScaleFactors=(PositiveWidth=1,NegativeWidth=1,PositiveHeight=1,NegativeHeight=1)
    FalloffSize=0.5
    FlipHeightScaling=False
    FlipWidthScaling=False
    RotationOffset=(X=90,Y=0,Z=0)
    OptionalParentItem=RigVMVariableNode() {variable=Spine}.Value
    DriverItem=RigVMVariableNode() {variable=Clavice}.Value
    ExecutePin=@RigVMFunction_ControlFlowBranch.Completed
   )
   Multiply = &RigVMFunction_MathVectorMul(B=(X=0,Y=18,Z=-15), A=RigVMVariableNode() {variable=reverseTran}.Value) {template="Multiply::Execute(in A,in B,out Result)"}
   > &RigUnit_ModifyTransforms(
    ItemToModify=((Item=(Type=Bone,Name="None"),Transform=(Rotation=(X=0,Y=0,Z=0,W=1),Translation=(X=0,Y=0,Z=0),Scale3D=(X=1,Y=1,Z=1))))
    WeightMinimum=0
    ItemToModify.0.Transform.Translation=@Multiply.Result
    ItemToModify.0.Item=RigVMVariableNode() {variable=DeformBone}.Value
    Weight=@n2.OutputParam
   )
   > RigVMFunctionReturnNode()
  ]}
 New Function_ContainedGraph {nodes [
   n1 = RigVMFunctionEntryNode()
   RigVMFunction_ControlFlowBranch = &RigVMFunction_ControlFlowBranch(ExecuteContext=@n1.ExecuteContext, Condition=@n1.Right)
   n2 = &RigUnit_SphericalPoseReader(
    ActiveRegionScaleFactors=(PositiveWidth=1,NegativeWidth=1,PositiveHeight=1,NegativeHeight=1)
    ActiveRegionSize=0.24
    Debug=(bDrawDebug=True,bDraw2D=False,bDrawLocalAxes=False,DebugScale=15,DebugSegments=20,DebugThickness=0.1)
    DriverAxis=(X=1,Y=0,Z=0)
    FalloffRegionScaleFactors=(PositiveWidth=1,NegativeWidth=1,PositiveHeight=1,NegativeHeight=1)
    FalloffSize=0.5
    FlipHeightScaling=False
    FlipWidthScaling=False
    RotationOffset=(X=90,Y=0,Z=100)
    ExecutePin=@RigVMFunction_ControlFlowBranch.Completed
    OptionalParentItem=@n1.thighBone
    DriverItem=@n1.KneeBone
   )
   Multiply = &RigVMFunction_MathVectorMul(A=(X=0,Y=1.5,Z=0), B=RigVMVariableNode() {variable=reverser}.Value) {template="Multiply::Execute(in A,in B,out Result)"}
   > &RigUnit_ModifyTransforms(
    ItemToModify=((Item=(Type=Bone,Name="None"),Transform=(Rotation=(X=0,Y=0,Z=0,W=1),Translation=(X=0,Y=2.5,Z=0),Scale3D=(X=1.2,Y=1,Z=1))))
    WeightMinimum=0
    ItemToModify.0.Transform.Translation=@Multiply.Result
    ItemToModify.0.Item=RigVMVariableNode() {variable=DeformBone}.Value
    Weight=@n2.OutputParam
   )
   > &RigUnit_ParentConstraint(
    AdvancedSettings=(InterpolationType=Average,RotationOrderForFilter=XZY)
    Parents=((Item=(Type=Bone,Name="None"),Weight=1),(Item=(Type=Bone,Name="None"),Weight=0.5))
    Weight=1
    Parents.1.Item=RigVMVariableNode() {variable=KneeBone}.Value
    Parents.0.Item=RigVMVariableNode() {variable=thighBone}.Value
    Child=RigVMVariableNode() {variable=BaseBone}.Value
   )
   > RigVMFunctionReturnNode()
   RigVMVariableNode(Value=(X=1,Y=1,Z=1), ExecuteContext=@RigVMFunction_ControlFlowBranch.False) {variable=reverser}
   RigVMVariableNode(Value=(X=-1,Y=-1,Z=-1), ExecuteContext=@RigVMFunction_ControlFlowBranch.True) {variable=reverser}
  ]}
 RigVMFunctionLibrary {nodes [
   BodySide_Deform = RigVMCollapseNode(Clavice=(Type=None,Name="None"), DeformBone=(Type=None,Name="None"), ExecuteContext=(), Right=False, Spine=(Type=None,Name="None"))
   ElbowIn_Deform = RigVMCollapseNode(
    BaseDone=(Type=None,Name="None")
    DeformBone=(Type=None,Name="None")
    DriverBone=(Type=None,Name="None")
    ExecuteContext=()
    parentBone=(Type=None,Name="None")
   )
   FootBelt_Deform = RigVMCollapseNode(DeformBone=(Type=None,Name="None"), ExecuteContext=(), footBone=(Type=None,Name="None"), legBone=(Type=None,Name="None"))
   HandUp_Deform = RigVMCollapseNode(
    DeformBone=(Type=None,Name="None")
    DeformBone_2=(Type=None,Name="None")
    ElbowBone=(Type=None,Name="None")
    ExecuteContext=()
    handBone=(Type=None,Name="None")
   )
   Heel_Deform = RigVMCollapseNode(
    ExecuteContext=()
    ParentBone=(Type=None,Name="None")
    boneBaseDeform=(Type=None,Name="None")
    boneDeform=(Type=None,Name="None")
    driverBone=(Type=None,Name="None")
   )
   KneeOut_Deform = RigVMCollapseNode(
    BaseBone=(Type=None,Name="None")
    DeformBone=(Type=None,Name="None")
    ExecuteContext=()
    KneeBone=(Type=None,Name="None")
    thighBone=(Type=None,Name="None")
   )
   ShoulderPad_Deform = RigVMCollapseNode(
    AimUpVector=(Type=None,Name="None")
    DeformBone=(Type=None,Name="None")
    DriverBone=(Type=None,Name="None")
    ExecuteContext=()
    Right=False
    TwistBone=(Type=None,Name="None")
    claviceBone=(Type=None,Name="None")
    spineBone=(Type=None,Name="None")
   )
   thighBelt_Deform = RigVMCollapseNode(ExecuteContext=(), Pelvis=(Type=None,Name="None"), beltBone=(Type=None,Name="None"), thigh=(Type=None,Name="None"))
  ]}
 RigVMModel {nodes [
   RigUnit_BeginExecution = &RigUnit_BeginExecution()
   &RigUnit_ModifyTransforms(
    ItemToModify=((Item=(Type=Bone,Name="calf_r"),Transform=(Rotation=(X=0,Y=0,Z=0,W=1),Translation=(X=0,Y=0,Z=0),Scale3D=(X=1,Y=1,Z=1))))
    Weight=1
    WeightMinimum=0
    ExecutePin=@RigUnit_BeginExecution.ExecutePin
   )
   > &RigUnit_ModifyTransforms(
    ItemToModify=((Item=(Type=Bone,Name="thigh_l"),Transform=(Rotation=(X=0,Y=0,Z=0,W=1),Translation=(X=0,Y=0,Z=0),Scale3D=(X=1,Y=1,Z=1))))
    Weight=1
    WeightMinimum=0
   )
   > &RigUnit_ModifyTransforms(
    ItemToModify=((Item=(Type=Bone,Name="upperarm_r"),Transform=(Rotation=(X=0,Y=0,Z=0,W=1),Translation=(X=0,Y=0,Z=0),Scale3D=(X=1,Y=1,Z=1))))
    Weight=1
    WeightMinimum=0
   )
   > &RigUnit_ModifyTransforms(
    ItemToModify=((Item=(Type=Bone,Name="clavicle_l"),Transform=(Rotation=(X=0,Y=0,Z=0,W=1),Translation=(X=0,Y=0,Z=0),Scale3D=(X=1,Y=1,Z=1))))
    Weight=1
    WeightMinimum=0
   )
   > FN_DeformRig_FNC_1 = RigVMFunctionReferenceNode(
    DeformRig_ Curve="(deformRig_l_upperarm_up_alpha_72_#1=False,deformRig_r_upperarm_up_alpha_73_#2=False,deformRig_l_upperarm_fr_alpha_74_#3=False,deformRig_r_upperarm_fr_alpha_75_CB51B8 ...(truncated)"
    DeformRig_Bool="(upperarm_1_#4=True,elbow_3_#5=True,wrist_7_#6=True,groin_8_#7=True,glute_17_#8=True,knee_9_40C408324CA261 ...(truncated)"
    DeformRig_Debug=(upperarm_debug_18_#4=False,wrist_debug_20_#6=False,lat_debug_26_#9=False,groinGlute_debug_27_#7=False)
    DeformRig_Value="(upperarm_up_18_#10=(Rotation=(X=0,Y=0,Z=0,W=1),Translation=(X=0,Y=0,Z=8),Scale3D=(X=0.7,Y=1,Z=1)),upperarm_fwd_fr_19_#11=(Rotation=(X=0,Y=0,Z=0,W=1),Translation=(X=0,Y=10,Z=0),Scale3D=(X=0.5,Y=1,Z= ...(truncated)"
   )
   > Heel_Deform_2 = RigVMFunctionReferenceNode(
    ParentBone=(Type=Bone,Name="calf_r")
    boneBaseDeform=(Type=Bone,Name="deform_heel_base_r")
    boneDeform=(Type=Bone,Name="deform_heel_r")
    driverBone=(Type=Bone,Name="foot_r")
   )
   > Heel_Deform_1_1 = RigVMFunctionReferenceNode(
    Left=true
    ParentBone=(Type=Bone,Name="calf_l")
    boneBaseDeform=(Type=Bone,Name="deform_heel_base_l")
    boneDeform=(Type=Bone,Name="deform_heel_l")
    driverBone=(Type=Bone,Name="foot_l")
   )
   > ShoulderPad_Deform_2 = RigVMFunctionReferenceNode(
    AimUpVector=(Type=Bone,Name="clavice_AimUp_l")
    DeformBone=(Type=Bone,Name="shoulderPad_l")
    DriverBone=(Type=Bone,Name="upperarm_l")
    Right=False
    TwistBone=(Type=Bone,Name="upperarm_twist_02_l")
    claviceBone=(Type=Bone,Name="clavicle_l")
    spineBone=(Type=Bone,Name="spine_05")
   )
   > ShoulderPad_Deform_1_1 = RigVMFunctionReferenceNode(
    AimUpVector=(Type=Bone,Name="clavice_AimUp_r")
    DeformBone=(Type=Bone,Name="shoulderPad_r")
    DriverBone=(Type=Bone,Name="upperarm_r")
    Right=true
    TwistBone=(Type=Bone,Name="upperarm_twist_02_r")
    claviceBone=(Type=Bone,Name="clavicle_r")
    spineBone=(Type=Bone,Name="spine_05")
   )
   > ElbowIn_Deform_2 = RigVMFunctionReferenceNode(
    BaseDone=(Type=Bone,Name="deform_elbow_in_base_l")
    DeformBone=(Type=Bone,Name="deform_elbow_in_l")
    DriverBone=(Type=Bone,Name="lowerarm_l")
    parentBone=(Type=Bone,Name="upperarm_twist_02_l")
   )
   > ElbowIn_Deform_1_1 = RigVMFunctionReferenceNode(
    BaseDone=(Type=Bone,Name="deform_elbow_in_base_r")
    DeformBone=(Type=Bone,Name="deform_elbow_in_r")
    DriverBone=(Type=Bone,Name="lowerarm_r")
    Right=true
    parentBone=(Type=Bone,Name="upperarm_twist_02_r")
   )
   > thighBelt_Deform_2 = RigVMFunctionReferenceNode(Pelvis=(Type=Bone,Name="pelvis"), beltBone=(Type=Bone,Name="deform_belt_l"), thigh=(Type=Bone,Name="thigh_l"))
   > thighBelt_Deform_1_1 = RigVMFunctionReferenceNode(Pelvis=(Type=Bone,Name="pelvis"), Right=true, beltBone=(Type=Bone,Name="deform_belt_r"), thigh=(Type=Bone,Name="thigh_r"))
   > /Script/ControlRig.RigUnit_PositionConstraintLocalSpaceOffset(
    Child=(Type=Bone,Name="deform_belt_base_m")
    Filter=(bX=True,bY=True,bZ=True)
    Parents=((Item=(Type=Bone,Name="deform_belt_l"),Weight=1),(Item=(Type=Bone,Name="deform_belt_r"),Weight=1))
    Weight=1
    bMaintainOffset=True
   ) {is=struct RigVMUnitNode}
   > &RigUnit_AimConstraintLocalSpaceOffset(
    AimAxis=(X=0,Y=-1,Z=0)
    Child=(Type=Bone,Name="deform_belt_m")
    Filter=(bX=True,bY=True,bZ=True)
    Parents=((Item=(Type=Bone,Name="deform_belt_l"),Weight=1))
    UpAxis=(X=1,Y=0,Z=0)
    Weight=1
    WorldUp=(Target=(X=1,Y=0,Z=0),Kind=Direction,Space=(Type=Bone,Name="pelvis"))
   )
   > HandUp_Deform_2 = RigVMFunctionReferenceNode(
    DeformBone=(Type=Bone,Name="deform_wrist_up_base_l")
    DeformBone_2=(Type=Bone,Name="deform_wrist_upSec_l")
    ElbowBone=(Type=Bone,Name="lowerarm_twist_01_l")
    handBone=(Type=Bone,Name="hand_l")
   )
   > HandUp_Deform_1_1 = RigVMFunctionReferenceNode(
    DeformBone=(Type=Bone,Name="deform_wrist_up_base_r")
    DeformBone_2=(Type=Bone,Name="deform_wrist_upSec_r")
    ElbowBone=(Type=Bone,Name="lowerarm_twist_01_r")
    Right=true
    handBone=(Type=Bone,Name="hand_r")
   )
   > FootBelt_Deform_2 = RigVMFunctionReferenceNode(DeformBone=(Type=Bone,Name="deform_footBelt_l"), footBone=(Type=Bone,Name="foot_l"), legBone=(Type=Bone,Name="calf_twist_01_l"))
   > FootBelt_Deform_1_1 = RigVMFunctionReferenceNode(DeformBone=(Type=Bone,Name="deform_footBelt_r"), Right=true, footBone=(Type=Bone,Name="foot_r"), legBone=(Type=Bone,Name="calf_twist_01_r"))
   > BodySide_Deform_2 = RigVMFunctionReferenceNode(Clavice=(Type=Bone,Name="clavicle_l"), DeformBone=(Type=Bone,Name="deform_BodySide_r"), Right=False, Spine=(Type=Bone,Name="spine_03"))
   > BodySide_Deform_1_1 = RigVMFunctionReferenceNode(Clavice=(Type=Bone,Name="clavicle_r"), DeformBone=(Type=Bone,Name="deform_BodySide_r1"), Right=true, Spine=(Type=Bone,Name="spine_03"))
   > KneeOut_Deform = RigVMFunctionReferenceNode(
    BaseBone=(Type=Bone,Name="deform_knee_out_base_l")
    DeformBone=(Type=Bone,Name="deform_knee_out_l")
    KneeBone=(Type=Bone,Name="calf_l")
    thighBone=(Type=Bone,Name="thigh_twist_01_l")
   )
   > KneeOut_Deform_1 = RigVMFunctionReferenceNode(
    BaseBone=(Type=Bone,Name="deform_knee_out_base_r")
    DeformBone=(Type=Bone,Name="deform_knee_out_r")
    KneeBone=(Type=Bone,Name="calf_r")
    Right=true
    thighBone=(Type=Bone,Name="thigh_twist_01_r")
   )
   &RigUnit_SetTransform(
    ExecutePin=()
    Item=(Type=Bone,Name="upperarm_r")
    Weight=1
    bInitial=False
    Value=&RigUnit_GetTransform(Item=(Type=Control,Name="upperarm_r_ctrl"), bInitial=False).Transform
   )
   > &RigUnit_SetTransform(
    Item=(Type=Bone,Name="lowerarm_l")
    Weight=1
    bInitial=False
    Value=&RigUnit_GetTransform(Item=(Type=Control,Name="lowerarm_l_ctrl"), bInitial=False).Transform
   )
   &RigUnit_SetTransform(
    ExecutePin=()
    Item=(Type=Bone,Name="calf_l")
    Weight=1
    bInitial=False
    Value=&RigUnit_GetTransform(Item=(Type=Control,Name="calf_l_ctrl"), bInitial=False).Transform
   )
   > &RigUnit_SetTransform(
    Item=(Type=Bone,Name="calf_r")
    Weight=1
    bInitial=False
    Value=&RigUnit_GetTransform(Item=(Type=Control,Name="calf_r_ctrl"), bInitial=False).Transform
   )
   &RigUnit_SetTransform(
    ExecutePin=()
    Item=(Type=Bone,Name="upperarm_l")
    Weight=1
    bInitial=False
    Value=&RigUnit_GetTransform(Item=(Type=Control,Name="upperarm_l_ctrl"), bInitial=False).Transform
   )
  ]}
]
hierarchy [
 root {loc=<0 0 0>, rot=<0 0 0 1>}
 pelvis {parent=root, loc=<0 -0.7071 93.1191>, rot=<0.0127 0.707 -0.0127 -0.707>}
 spine_01 {parent=pelvis, loc=<6.7494 -0.6388 0>, rot=<0 0 0.034 -0.9994>}
 spine_02 {parent=spine_01, loc=<7.2518 -0.035 0>, rot=<0 0 0.0238 -0.9997>}
 spine_03 {parent=spine_02, loc=<6.0104 0 0>, rot=<0 0 0.0149 -0.9999>}
 spine_04 {parent=spine_03, loc=<8.1024 0 0>, rot=<0 0 -0.0132 -0.9999>}
 spine_05 {parent=spine_04, loc=<10.0146 0 0>, rot=<0 0 -0.0015 -1>}
 attach_cape {parent=spine_05, loc=<9.017 -11.714 0>, rot=<0.0616 -0.7044 -0.0616 -0.7044>}
 attach_backpack {parent=spine_05, loc=<-0.123 -13.473 0>, rot=<-0.0093 -0.707 0.0093 -0.707>}
 clavicle_l {parent=spine_05, loc=<4.3575 -0.1466 -2.3223>, rot=<0.0465 -0.7289 -0.009 -0.6829>}
 upperarm_l {parent=clavicle_l, loc=<15.5111 0 0>, rot=<0.0226 -0.3234 -0.0088 -0.9459>}
 lowerarm_l {parent=upperarm_l, loc=<26.6514 0 0>, rot=<0.0391 -0.0099 -0.2453 -0.9686>}
 hand_l {parent=lowerarm_l, loc=<22.3491 0 0>, rot=<-0.7081 -0.0507 0.0975 -0.6975>}
 weapon_l {parent=hand_l, loc=<-0.1974 -3.4111 0.3811>, rot=<0 0 0 1>}
 hand_attach_l {parent=hand_l, loc=<8.5 -3 0>, rot=<0.7071 0.7071 0 0>}
 thumb_01_l {parent=hand_l, loc=<1.3118 -0.3191 -2.459>, rot=<0.6546 -0.4042 0.0664 -0.6354>}
 thumb_02_l {parent=thumb_01_l, loc=<5.9336 -0.0448 0.0401>, rot=<-0.0037 0.0213 0.298 -0.9543>}
 thumb_03_l {parent=thumb_02_l, loc=<2.2223 0.3682 -0.0907>, rot=<0.0118 -0.013 0.0066 -0.9998>}
 index_metacarpal_l {parent=hand_l, loc=<1.6535 0.0183 -2.96>, rot=<0.0158 -0.0124 -0.0173 -0.9996>}
 index_01_l {parent=index_metacarpal_l, loc=<8.3008 0.0357 -0.0524>, rot=<-0.0044 0.0485 0.0924 -0.9945>}
 index_02_l {parent=index_01_l, loc=<3.3587 -0.3004 0.0452>, rot=<-0.002 0.0027 0.3225 -0.9466>}
 index_03_l {parent=index_02_l, loc=<2.5578 -0.123 -0.0252>, rot=<-0.002 0.0067 0.2558 -0.9667>}
 middle_metacarpal_l {parent=hand_l, loc=<2.1856 -0.1725 -0.5984>, rot=<0.006 0.0168 -0.0329 -0.9993>}
 middle_01_l {parent=middle_metacarpal_l, loc=<7.3275 -0.2447 0.0129>, rot=<-0.0208 0.0042 0.0858 -0.9961>}
 middle_02_l {parent=middle_01_l, loc=<4.0123 0.1401 0.1175>, rot=<-0.0218 0.0258 0.3512 -0.9357>}
 middle_03_l {parent=middle_02_l, loc=<2.6077 0.0534 0.1373>, rot=<0.0082 0.0188 0.1708 -0.9851>}
 ring_metacarpal_l {parent=hand_l, loc=<2.2074 -0.3625 1.3638>, rot=<-0.0284 0.0722 -0.0458 -0.9959>}
 ring_01_l {parent=ring_metacarpal_l, loc=<7.0295 -0.6197 -0.1446>, rot=<-0.0489 -0.0252 0.1892 -0.9804>}
 ring_02_l {parent=ring_01_l, loc=<3.6174 0.4647 -0.1027>, rot=<0 0 0.2806 -0.9598>}
 ring_03_l {parent=ring_02_l, loc=<2.4152 0.114 0.0199>, rot=<0 0 0.1189 -0.9929>}
 pinky_metacarpal_l {parent=hand_l, loc=<1.7616 -0.9753 3.3814>, rot=<-0.0397 0.0814 -0.069 -0.9935>}
 pinky_01_l {parent=pinky_metacarpal_l, loc=<7.1709 -0.4831 -0.1094>, rot=<-0.0654 -0.0035 0.2242 -0.9723>}
 pinky_02_l {parent=pinky_01_l, loc=<2.8274 0.0893 -0.1043>, rot=<0.0037 0.0011 0.2854 -0.9584>}
 pinky_03_l {parent=pinky_02_l, loc=<2.2046 0.0307 -0.0078>, rot=<-0.0002 0.0098 0.1445 -0.9895>}
 deform_wrist_dn_base_l {parent=hand_l, loc=<0 -2.2342 0>, rot=<0 0 0 -1>}
 deform_wrist_dn_l {parent=deform_wrist_dn_base_l, loc=<0 0 0>, rot=<0 0 0 1>}
 deform_wrist_up_base_l {parent=hand_l, loc=<0 2.2682 0>, rot=<0 0 0 -1>}
 deform_wrist_up_l {parent=deform_wrist_up_base_l, loc=<0 0 0>, rot=<0 0 0 1>}
 deform_wrist_upSec_l {parent=deform_wrist_up_base_l, loc=<0 -0.5 0>, rot=<0 0 0 1>}
 lowerarm_twist_02_l {parent=lowerarm_l, loc=<9.6199 0 0>, rot=<0 0 0 -1>}
 lowerarm_twist_01_l {parent=lowerarm_l, loc=<17.8549 0 0>, rot=<0 0 0 -1>}
 deform_elbow_in_base_l {parent=lowerarm_l, loc=<0.2112 5.259 0.8274>, rot=<0 0 0 -1>}
 deform_elbow_in_l {parent=deform_elbow_in_base_l, loc=<0 0 0>, rot=<0 0 0 1>}
 deform_elbow_out_base_l {parent=lowerarm_l, loc=<0 -4 0>, rot=<0 0 0 -1>}
 deform_elbow_out_l {parent=deform_elbow_out_base_l, loc=<0 0 0>, rot=<0 0 0 1>}
 upperarm_twist_01_l {parent=upperarm_l, loc=<0.0041 0 0>, rot=<0 0 0 1>}
 deform_upperarm_bk_base_l {parent=upperarm_twist_01_l, loc=<1.0813 -7.2193 -1.9453>, rot=<0 0 0 1>}
 deform_upperarm_bk_l {parent=deform_upperarm_bk_base_l, loc=<0 0 0>, rot=<0 0 0 1>}
 deform_upperarm_fr_base_l {parent=upperarm_twist_01_l, loc=<-0.3691 7.0506 -2.765>, rot=<0 0 0 1>}
 deform_upperarm_fr_l {parent=deform_upperarm_fr_base_l, loc=<0 0 0>, rot=<0 0 0 1>}
 deform_upperarm_up_base_l {parent=upperarm_twist_01_l, loc=<-3.2216 -1.1391 3.3311>, rot=<0 0 0 1>}
 deform_upperarm_up_l {parent=deform_upperarm_up_base_l, loc=<0 0 0>, rot=<0 0 0 1>}
 shoulderPad_base_l {parent=upperarm_twist_01_l, loc=<-0.0041 0 0>, rot=<-0.019 0.829 -0.0037 -0.559>}
 shoulderPad_l {parent=shoulderPad_base_l, loc=<0 0 0>, rot=<0 0 0 1>}
 upperarm_twist_02_l {parent=upperarm_l, loc=<13.7941 0 0>, rot=<0 0 0 1>}
 deform_bicep_base_l {parent=upperarm_twist_02_l, loc=<0 5.86 0>, rot=<0 0 -0.1305 -0.9914>}
 deform_bicep_l {parent=deform_bicep_base_l, loc=<0 0 0>, rot=<0 0 0 1>}
 clavicle_l_Vector {parent=clavicle_l, type=Null, loc=<15 0 30>, rot=<0 0 0 1>}
 clavicle_l_ctrl {controlType=EULER_TRANSFORM, parent=clavicle_l, type=Control, loc=<0 0 0>, rot=<0 0 0 1>}
 clavice_AimUp_l {parent=clavicle_l, loc=<15 0 30>, rot=<0 0 0 1>}
 neck_01 {parent=spine_05, loc=<11.3138 0.1529 0>, rot=<0 0 -0.2003 -0.9797>}
 neck_02 {parent=neck_01, loc=<4.5627 0 0>, rot=<0 0 0.0504 -0.9987>}
 head {parent=neck_02, loc=<5.3914 0 0>, rot=<0 0 0.115 -0.9934>}
 hair_l_01 {parent=head, loc=<9.1649 -1.6624 -5.5838>, rot=<0.2849 -0.4143 0.1862 -0.8441>}
 hair_l_02 {parent=hair_l_01, loc=<6.0474 0 0>, rot=<0 0 0 1>}
 hair_r_01 {parent=head, loc=<9.1648 -1.6624 5.5837>, rot=<-0.4143 -0.2849 0.8441 0.1862>}
 hair_r_02 {parent=hair_r_01, loc=<-6.0477 0 -0.0004>, rot=<0 0 0 1>}
 clavicle_r {parent=spine_05, loc=<4.3575 -0.1466 2.3223>, rot=<0.7289 0.0465 -0.6829 0.009>}
 upperarm_r {parent=clavicle_r, loc=<-15.5111 0 0>, rot=<0.0226 -0.3234 -0.0088 -0.9459>}
 lowerarm_r {parent=upperarm_r, loc=<-26.6514 0 0>, rot=<0.0391 -0.0099 -0.2453 -0.9686>}
 hand_r {parent=lowerarm_r, loc=<-22.3491 0 0>, rot=<-0.7081 -0.0507 0.0975 -0.6975>}
 hand_attach_r {parent=hand_r, loc=<-8.5001 2.9999 0>, rot=<0 0 -0.7071 -0.7071>}
 thumb_01_r {parent=hand_r, loc=<-1.3118 0.3191 2.459>, rot=<0.6546 -0.4042 0.0664 -0.6354>}
 thumb_02_r {parent=thumb_01_r, loc=<-5.9336 0.0448 -0.0401>, rot=<-0.0037 0.0213 0.298 -0.9543>}
 thumb_03_r {parent=thumb_02_r, loc=<-2.2223 -0.3682 0.0907>, rot=<0.0118 -0.013 0.0066 -0.9998>}
 index_metacarpal_r {parent=hand_r, loc=<-1.6535 -0.0183 2.96>, rot=<0.0158 -0.0124 -0.0173 -0.9996>}
 index_01_r {parent=index_metacarpal_r, loc=<-8.3008 -0.0357 0.0524>, rot=<-0.0044 0.0485 0.0924 -0.9945>}
 index_02_r {parent=index_01_r, loc=<-3.3587 0.3004 -0.0452>, rot=<-0.002 0.0027 0.3225 -0.9466>}
 index_03_r {parent=index_02_r, loc=<-2.5578 0.123 0.0252>, rot=<-0.002 0.0067 0.2558 -0.9667>}
 middle_metacarpal_r {parent=hand_r, loc=<-2.1856 0.1725 0.5984>, rot=<0.006 0.0168 -0.0329 -0.9993>}
 middle_01_r {parent=middle_metacarpal_r, loc=<-7.3275 0.2447 -0.0129>, rot=<-0.0208 0.0042 0.0858 -0.9961>}
 middle_02_r {parent=middle_01_r, loc=<-4.0123 -0.1401 -0.1175>, rot=<-0.0218 0.0258 0.3512 -0.9357>}
 middle_03_r {parent=middle_02_r, loc=<-2.6077 -0.0534 -0.1373>, rot=<0.0082 0.0188 0.1708 -0.9851>}
 ring_metacarpal_r {parent=hand_r, loc=<-2.2074 0.3625 -1.3638>, rot=<-0.0284 0.0722 -0.0458 -0.9959>}
 ring_01_r {parent=ring_metacarpal_r, loc=<-7.0295 0.6197 0.1446>, rot=<-0.0489 -0.0252 0.1892 -0.9804>}
 ring_02_r {parent=ring_01_r, loc=<-3.6174 -0.4647 0.1027>, rot=<0 0 0.2806 -0.9598>}
 ring_03_r {parent=ring_02_r, loc=<-2.4152 -0.114 -0.0199>, rot=<0 0 0.1189 -0.9929>}
 pinky_metacarpal_r {parent=hand_r, loc=<-1.7616 0.9753 -3.3814>, rot=<-0.0397 0.0814 -0.069 -0.9935>}
 pinky_01_r {parent=pinky_metacarpal_r, loc=<-7.1709 0.4831 0.1094>, rot=<-0.0654 -0.0035 0.2242 -0.9723>}
 pinky_02_r {parent=pinky_01_r, loc=<-2.8274 -0.0893 0.1043>, rot=<0.0037 0.0011 0.2854 -0.9584>}
 pinky_03_r {parent=pinky_02_r, loc=<-2.2046 -0.0307 0.0078>, rot=<-0.0002 0.0098 0.1445 -0.9895>}
 weapon_r {parent=hand_r, loc=<0.1974 3.4111 -0.3811>, rot=<0 0 0 1>}
 deform_wrist_dn_base_r {parent=hand_r, loc=<-0.0004 2.2345 -0.0001>, rot=<0 0 0 -1>}
 deform_wrist_dn_r {parent=deform_wrist_dn_base_r, loc=<0 0 0>, rot=<0 0 0 1>}
 deform_wrist_up_base_r {parent=hand_r, loc=<0.0003 -2.2684 0.0001>, rot=<0 0 0 -1>}
 deform_wrist_up_r {parent=deform_wrist_up_base_r, loc=<0 0 0>, rot=<0 0 0 1>}
 deform_wrist_upSec_r {parent=deform_wrist_up_base_r, loc=<0 0.5 0>, rot=<0 0 0 1>}
 lowerarm_twist_02_r {parent=lowerarm_r, loc=<-9.6199 0 0>, rot=<0 0 0 -1>}
 lowerarm_twist_01_r {parent=lowerarm_r, loc=<-17.8549 0 0>, rot=<0 0 0 -1>}
 deform_elbow_in_base_r {parent=lowerarm_r, loc=<-0.2111 -5.2591 -0.8275>, rot=<0 0 0 -1>}
 deform_elbow_in_r {parent=deform_elbow_in_base_r, loc=<-0.0001 0 0.0002>, rot=<0 0 0 1>}
 deform_elbow_out_base_r {parent=lowerarm_r, loc=<0 4 -0.0001>, rot=<0 0 0 -1>}
 deform_elbow_out_r {parent=deform_elbow_out_base_r, loc=<0 0 0>, rot=<0 0 0 1>}
 upperarm_twist_01_r {parent=upperarm_r, loc=<-0.0041 0 0>, rot=<0 0 0 -1>}
 deform_upperarm_bk_base_r {parent=upperarm_twist_01_r, loc=<-1.0814 7.2192 1.9455>, rot=<0 0 0 1>}
 deform_upperarm_bk_r {parent=deform_upperarm_bk_base_r, loc=<0 0 0>, rot=<0 0 0 1>}
 deform_upperarm_fr_base_r {parent=upperarm_twist_01_r, loc=<0.3689 -7.0506 2.7652>, rot=<0 0 0 1>}
 deform_upperarm_fr_r {parent=deform_upperarm_fr_base_r, loc=<0 0 0>, rot=<0 0 0 1>}
 deform_upperarm_up_base_r {parent=upperarm_twist_01_r, loc=<3.2214 1.1391 -3.3308>, rot=<0 0 0 1>}
 deform_upperarm_up_r {parent=deform_upperarm_up_base_r, loc=<0 0 0>, rot=<0 0 0 -1>}
 shoulderPad_base_r {parent=upperarm_twist_01_r, loc=<0.0041 0 0>, rot=<-0.019 0.829 -0.0037 -0.559>}
 shoulderPad_r {parent=shoulderPad_base_r, loc=<0 0 0>, rot=<0 0 0 1>}
 upperarm_twist_02_r {parent=upperarm_r, loc=<-13.7941 0 0>, rot=<0 0 0 -1>}
 deform_bicep_base_r {parent=upperarm_twist_02_r, loc=<0.0002 -5.86 -0.0002>, rot=<0 0 -0.1305 -0.9914>}
 deform_bicep_r {parent=deform_bicep_base_r, loc=<0 0 0>, rot=<0 0 0 1>}
 clavicle_r_Vector {parent=clavicle_r, type=Null, loc=<-15 0 30>, rot=<0 0 0 1>}
 clavicle_r_ctrl {controlType=EULER_TRANSFORM, parent=clavicle_r, type=Control, loc=<0 0 0>, rot=<0 0 0 1>}
 clavice_AimUp_r {parent=clavicle_r, loc=<-15 0 30>, rot=<0 0 0 1>}
 deform_pec_aim_base_l {parent=spine_04, loc=<7.8732 12.6388 -2.4123>, rot=<0.0102 -0.4602 0.3979 -0.7936>}
 deform_pec_aim_l {parent=deform_pec_aim_base_l, loc=<0 0 0>, rot=<0 0 0 1>}
 deform_pec_base_l {parent=deform_pec_aim_l, loc=<10.766 0 0>, rot=<0 0 0 1>}
 deform_pec_l {parent=deform_pec_base_l, loc=<0.055 0 0>, rot=<0 0 0 1>}
 deform_pec_aim_base_r {parent=spine_04, loc=<7.8727 12.6388 2.4123>, rot=<-0.4602 -0.0102 0.7936 0.3979>}
 deform_pec_aim_r {parent=deform_pec_aim_base_r, loc=<0 0 0>, rot=<0 0 0 1>}
 deform_pec_base_r {parent=deform_pec_aim_r, loc=<-10.7662 -0.0006 -0.0007>, rot=<0 0 0 1>}
 deform_pec_r {parent=deform_pec_base_r, loc=<0 0 0>, rot=<0 0 0 1>}
 deform_BodySide_r {parent=spine_03, loc=<1.3585 -3.123 -14.0261>, rot=<-0.6989 -0.1074 0.0213 -0.7068>}
 deform_BodySide_r1 {parent=spine_03, loc=<1.3583 -3.1231 14.0261>, rot=<-0.1074 0.6989 0.7068 0.0213>}
 thigh_l {parent=pelvis, loc=<-2.5693 -0.0109 -7.7501>, rot=<0.0012 0.0572 0.0155 -0.9982>}
 calf_l {parent=thigh_l, loc=<-39.0484 0 0>, rot=<0.0006 -0.0161 -0.0391 -0.9991>}
 foot_l {parent=calf_l, loc=<-42.561 0.1077 0.4292>, rot=<0.0418 -0.0455 0.0343 -0.9975>}
 ball_l {parent=foot_l, loc=<-6.1599 13.8264 0.6806>, rot=<0.0001 -0.0001 -0.7063 -0.7079>}
 deform_footBelt_base_l {parent=foot_l, loc=<1.9749 5.7077 0.5299>, rot=<0 0 -0.2774 -0.9608>}
 deform_footBelt_l {parent=deform_footBelt_base_l, loc=<0 0 0>, rot=<0 0 0 1>}
 calf_twist_02_l {parent=calf_l, loc=<-18.8064 -0.0967 0.0397>, rot=<0.0237 0.0095 -0.0035 -0.9997>}
 calf_twist_01_l {parent=calf_l, loc=<-35.813 -0.0967 0.0397>, rot=<0.0237 0.0095 -0.0035 -0.9997>}
 deform_heel_base_l {parent=calf_twist_01_l, loc=<-4 -6 0.4>, rot=<0.8327 -0.5507 -0.0383 0.0447>}
 deform_heel_l {parent=deform_heel_base_l, loc=<0 0 0>, rot=<0 0 0 1>}
 deform_knee_in_base_l {parent=calf_l, loc=<1.298 -5.2957 -0.0496>, rot=<0 0 0 -1>}
 deform_knee_in_l {parent=deform_knee_in_base_l, loc=<0 0 0>, rot=<0 0 0 1>}
 deform_knee_out_base_l {parent=calf_l, loc=<0 4 0>, rot=<0 0 0 -1>}
 deform_knee_out_l {parent=deform_knee_out_base_l, loc=<0 0 0>, rot=<0 0 0 1>}
 thigh_twist_01_l {parent=thigh_l, loc=<-21.6303 0 0>, rot=<0.0013 0 0 -1>}
 deform_groin_base_l {parent=thigh_l, loc=<-2.1908 8.1142 -1.7442>, rot=<0 0 0 -1>}
 deform_groin_l {parent=deform_groin_base_l, loc=<0 0 0>, rot=<0 0 0 1>}
 thigh_r {parent=pelvis, loc=<-2.5693 -0.0109 7.7501>, rot=<0.0572 -0.0012 0.9982 0.0155>}
 calf_r {parent=thigh_r, loc=<39.0484 0 0>, rot=<0.0006 -0.0161 -0.0391 -0.9991>}
 foot_r {parent=calf_r, loc=<42.561 -0.1077 -0.4292>, rot=<0.0418 -0.0455 0.0343 -0.9975>}
 ball_r {parent=foot_r, loc=<6.1599 -13.8264 -0.6806>, rot=<0.0001 -0.0001 -0.7063 -0.7079>}
 deform_footBelt_base_r {parent=foot_r, loc=<-1.9748 -5.7077 -0.5299>, rot=<0 0 -0.2774 -0.9608>}
 deform_footBelt_r {parent=deform_footBelt_base_r, loc=<0 0 0>, rot=<0 0 0 1>}
 calf_twist_02_r {parent=calf_r, loc=<18.8064 0.0967 -0.0397>, rot=<0.0237 0.0095 -0.0035 -0.9997>}
 calf_twist_01_r {parent=calf_r, loc=<35.813 0.0967 -0.0397>, rot=<0.0237 0.0095 -0.0035 -0.9997>}
 deform_heel_base_r {parent=calf_twist_01_r, loc=<4 5.9999 -0.4>, rot=<0.8327 -0.5507 -0.0383 0.0447>}
 deform_heel_r {parent=deform_heel_base_r, loc=<0 0 0>, rot=<0 0 0 -1>}
 deform_knee_in_base_r {parent=calf_r, loc=<-1.298 5.2957 0.0496>, rot=<0 0 0 -1>}
 deform_knee_in_r {parent=deform_knee_in_base_r, loc=<0 0 0>, rot=<0 0 0 1>}
 deform_knee_out_base_r {parent=calf_r, loc=<0 -4 0>, rot=<0 0 0 -1>}
 deform_knee_out_r {parent=deform_knee_out_base_r, loc=<0 0 0>, rot=<0 0 0 1>}
 thigh_twist_01_r {parent=thigh_r, loc=<21.6303 0 0>, rot=<0.0013 0 0 -1>}
 deform_groin_base_r {parent=thigh_r, loc=<2.1908 -8.1142 1.7442>, rot=<0 0 0 -1>}
 deform_groin_r {parent=deform_groin_base_r, loc=<0 0 0>, rot=<0 0 0 1>}
 deform_glute_base_l {parent=pelvis, loc=<-8.8064 -8.8222 -8.4287>, rot=<0.0275 0.0501 -0.4472 -0.8926>}
 deform_glute_l {parent=deform_glute_base_l, loc=<0 0 0>, rot=<0 0 0 1>}
 deform_glute_base_r {parent=pelvis, loc=<-8.8062 -8.8222 8.4287>, rot=<0.0501 -0.0275 0.8926 -0.4472>}
 deform_glute_r {parent=deform_glute_base_r, loc=<0 0 0>, rot=<0 0 0 1>}
 deform_belt_base_l {parent=pelvis, loc=<5.927 6.9023 -8.4773>, rot=<0.1832 0.0079 -0.062 -0.9811>}
 deform_belt_l {parent=deform_belt_base_l, loc=<0 0 0>, rot=<0 0 0 1>}
 deform_belt_base_r {parent=pelvis, loc=<5.927 6.9023 8.4773>, rot=<0.0079 -0.1832 0.9811 -0.062>}
 deform_belt_r {parent=deform_belt_base_r, loc=<0 0 0>, rot=<0 0 0 -1>}
 deform_belt_base_m {parent=pelvis, loc=<6.0523 10.3976 0>, rot=<-0.5562 -0.0038 0.0057 -0.831>}
 deform_belt_m {parent=deform_belt_base_m, loc=<0 0 0>, rot=<0 0 0 1>}
 attach {parent=root, loc=<0 0 0>, rot=<0 0 0 1>}
 ik_hand_root {parent=root, loc=<0 0 0>, rot=<0 0 0 1>}
 ik_hand_gun {parent=ik_hand_root, loc=<-51.9727 4.0883 102.6935>, rot=<0.6084 0.1793 0.4613 -0.6205>}
 ik_hand_r {parent=ik_hand_gun, loc=<0 0 0>, rot=<0 0 0 1>}
 ik_hand_l {parent=ik_hand_gun, loc=<53.0285 82.1781 35.2051>, rot=<-0.5102 -0.7906 -0.3387 0>}
 root_fp {parent=root, loc=<0 0 0>, rot=<0 0 0 1>}
 attach_fp {parent=root_fp, loc=<0 0 0>, rot=<0 0 0 1>}
 camera_root {parent=root_fp, loc=<0 0 0>, rot=<0 0 0 1>}
 camera {parent=camera_root, loc=<0 0 0>, rot=<0 0 0 1>}
 attach_emote_01 {parent=root, loc=<0 0 0>, rot=<0 0 0 1>}
 attach_emote_02 {parent=root, loc=<0 0 0>, rot=<0 0 0 1>}
 attach_emote_03 {parent=root, loc=<0 0 0>, rot=<0 0 0 1>}
 warp_point_01 {parent=root, loc=<0 0 0>, rot=<0 0 0 1>}
 warp_point_02 {parent=root, loc=<0 0 0>, rot=<0 0 0 1>}
 warp_point_03 {parent=root, loc=<0 0 0>, rot=<0 0 0 1>}
 ik_foot_root {parent=root, loc=<0 0 0>, rot=<0 0 0 1>}
 ik_foot_l {parent=ik_foot_root, loc=<15.2608 -4.4176 9.4432>, rot=<0.0357 0.7045 0.0284 -0.7082>}
 ik_foot_r {parent=ik_foot_root, loc=<-15.2608 -4.4176 9.4432>, rot=<0.7082 0.0284 -0.7045 0.0357>}
 upperarm_l_ctrl {controlType=EULER_TRANSFORM, type=Control, loc=<0 0 0>, rot=<0 0 0 1>}
 upperarm_r_ctrl {controlType=EULER_TRANSFORM, type=Control, loc=<0 0 0>, rot=<0 0 0 1>}
]
