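# Physics asset for the face accessory mesh of the OliveStomp medium male player character, defining simulation bodies for the head and two hair chains.
# The head and each of the two left and two right hair bones get a small capsule body.
# Constraints link head to first to second bone on each side, with identical moderate swing limits and tight twist limits.
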PhysicsAsset /BRCosmetics/Characters/Player/Male/Medium/Bodies/M_MED_OliveStomp/Meshes/Parts/M_MED_OliveStomp_FaceAcc_Physics
bodies {
 hair_l_01 [{shapeType=Capsule, length=1, radius=1, shapeName=hair_l_01_capsule, center=<1.2975 0 0>, rotation=<rot 90 0 0>}]
 hair_l_02 [{shapeType=Capsule, length=1, radius=1, shapeName=hair_l_02_capsule, center=<1.623 0 0>, rotation=<rot 90 0 0>}]
 hair_r_01 [{shapeType=Capsule, length=1, radius=1, shapeName=hair_r_01_capsule, center=<-1.4125 0 0>, rotation=<rot 90 0 0>}]
 hair_r_02 [{shapeType=Capsule, length=1, radius=1, shapeName=hair_r_02_capsule, center=<-1.9232 0 0>, rotation=<rot 90 0 0>}]
 head [{shapeType=Capsule, length=1, radius=1, shapeName=head_capsule}]
}
constraints [
 {bone1Name=hair_l_01, bone2Name=hair_l_02, swing1LimitDegrees=30, swing2LimitDegrees=30, twistLimitDegrees=5}
 {bone1Name=hair_r_01, bone2Name=hair_r_02, swing1LimitDegrees=30, swing2LimitDegrees=30, twistLimitDegrees=5}
 {bone1Name=head, bone2Name=hair_r_01, swing1LimitDegrees=30, swing2LimitDegrees=30, twistLimitDegrees=5}
 {bone1Name=head, bone2Name=hair_l_01, swing1LimitDegrees=30, swing2LimitDegrees=30, twistLimitDegrees=5}
]
