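# Animation Blueprint for the OliveStomp male character's face accessory part, whose output pose is copied directly from the attached parent mesh.
# A head-space RigidBody chain with its own physics asset is present but not wired into the output pose.
# Each update sets that node's gravity override through a deprecated function.
# It also derives an external force from previous-minus-current head position over delta time, scaled, so it opposes component-space head movement.
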AnimBlueprint /BRCosmetics/Characters/Player/Male/Medium/Bodies/M_MED_OliveStomp/Meshes/Parts/M_MED_OliveStomp_FaceAcc_AnimBP
target_skeleton=&/M_MED_OliveStomp_FaceAcc_Skeleton
graphs [
 AnimGraph {nodes [
   n1 = RigidBody!(
    ComponentPose=Local To Component()
    Alpha
    OverrideWorldGravity:Vector=Get GravityOverride()
    ExternalForce=Get headComponentVelocity() -> (headComponentVelocity:Vector)
   ) -> (Pose) {
    bClampLinearTranslationLimitToRefPose=true
    bOverrideWorldGravity=true
    overridePhysicsAsset=&/M_MED_OliveStomp_FaceAcc_Physics
    simulationSpace=BaseBoneSpace
    alphaBoolBlend {blendOption=Linear}
    baseBoneRef {boneName=head}
    componentAppliedLinearAccClamp=<3000 3000 3000>
    componentLinearAccScale=<0.3 0.3 0.3>
    componentLinearVelScale=<0.3 0.3 0.3>
    simSpaceSettings {maxLinearAcceleration=0, maxLinearVelocity=0}
   }
   Component To Local(ComponentPose=@n1)
   Output Pose(Result=Copy Pose From Mesh() {bUseAttachedParent=true}) {layerGroup=None}
  ]}
 EventGraph {nodes [
   Event BlueprintUpdateAnimation()
   > n1 = Gravity Override State Function DEPRECATED(CharacterPartAnimInstance=Self-Reference(), GameTimeInSeconds=GetGameTimeInSeconds(), GravityOverrideParams=Get Gravity Override Params())
   > Set GravityOverride(GravityOverride=@n1)
   n2 = GetDeltaSeconds!(self:Anim Instance Object Reference) -> (ReturnValue:float)
   n3 = GetBoneTransform!(
    self:Skinned Mesh Component Object Reference=GetPartSkeletalMeshComponent()
    InBoneName:Name=head
    TransformSpace:ERelativeTransformSpace Enum=RTS_Component
   ) -> (ReturnValue_Location:Vector, ReturnValue_Rotation:Rotator, ReturnValue_Scale:Vector)
   n4 = vector - vector(A=Get headPreviousLocation() -> (headPreviousLocation:Vector), B=@n3)
   > Set headComponentVelocity(headComponentVelocity:Vector=vector * vector(A=vector / vector(A:Vector=@n4, B:float=@n2) -> (ReturnValue:Vector), B:Vector=Get headVelocityForce() -> (headVelocityForce:Vector)) -> (ReturnValue=~)) -> (Output_Get:Vector)
   > Set headPreviousLocation(headPreviousLocation:Vector=@n3) -> (Output_Get:Vector)
  ]}
]
variables [
 Gravity Override Params {type=GravityOverrideParamsStruct}
 GravityOverride {type=Vector}
 headComponentVelocity {type=Vector}
 headDeltaMove {type=Vector}
 headPreviousLocation {type=Vector}
 headVelocityForce {type=Vector, default=<400 400 400>}
]
